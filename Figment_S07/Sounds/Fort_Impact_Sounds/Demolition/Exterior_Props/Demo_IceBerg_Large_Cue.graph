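# Sound cue for the demolition sound of a large iceberg exterior prop.
# A single wave player plays one iceberg break sample, lowered slightly in pitch and volume.
# It uses the large-object demolition destroy attenuation and the explosions sound class.
# An object-impact concurrency set manages simultaneous instances.
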SoundCue /Figment_S07/Sounds/Fort_Impact_Sounds/Demolition/Exterior_Props/Demo_IceBerg_Large_Cue
root=WavePlayer_6
nodes [{id=WavePlayer_6, properties {soundWaveAssetPtr=&/Figment_Iceberg_Break_01}}]
properties {
 attenuationSettings=/Game/Sounds/Attenuation/Fort_Demolition_Destroy_Large_Sized_Object_Att
 maxDistance=6250
 pitchMultiplier=0.9
 soundClassObject=/Game/Sounds/FortSoundClassesAndModes/Fort_Events_Explosions
 volumeMultiplier=0.7
 concurrencyOverrides {resolutionRule=PreventNew}
 concurrencySet [&CONC_Object_Impacts_FourMax]
}
